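# Prototype base behavior tree for WildEstate AI that moves the agent to a goal actor or idles, using the BB_BuzzCool_Clone blackboard.
# The root selector first tries the Follow or Wait branch, gated on GoalActor, where a Move To tracks the moving actor and accepts partial paths.
# When that branch is blocked or fails, the tree falls back to a short Wait.
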BehaviorTree /BuzzCool/AI/BehaviorTrees/Proto/BT_WildEstate_Base
blackboard=/BuzzCool/AI/BehaviorTrees/BB_BuzzCool_Clone
root {
 name=Selector
 class=BTComposite_Selector
 children [
  Follow or Wait {
   class=BTComposite_Selector
   children [Move To {
     class=BTTask_MoveTo
     children []
     decorators []
     properties {
      bIgnoreRestartSelf=false
      nodeName=Move To
      acceptableRadius {defaultValue=5, key=None}
      bAllowPartialPath {defaultValue=true, key=None}
      bAllowStrafe {defaultValue=false, key=None}
      bProjectGoalLocation {defaultValue=true, key=None}
      bReachTestIncludesAgentRadius {defaultValue=true, key=None}
      bReachTestIncludesGoalRadius {defaultValue=true, key=None}
      bRequireNavigableEndLocation {defaultValue=true, key=None}
      bStartFromPreviousPath {defaultValue=false, key=None}
      bTrackMovingGoal {defaultValue=true, key=None}
      blackboardKey {bNoneIsAllowedValue=false, selectedKeyName=GoalActor}
      filterClass {baseClass=&NavigationQueryFilter, defaultValue=null, key=None}
      observedBlackboardValueTolerance {defaultValue=4.75, key=None}
     }
     services []
    }]
   decorators [Has Goal Actor? {
     class=BTDecorator_Blackboard
     properties {
      arithmeticOperation=Equal
      intValue=0
      nodeName=Has Goal Actor?
      notifyObserver=ValueChange
      stringValue=""
      blackboardKey {selectedKeyName=GoalActor}
     }
    }]
   properties {
    nodeName=Follow or Wait
    children [{childComposite=null, childTask=&self:BTTask_MoveTo_2, decorators []}]
    services []
   }
   services []
  }
  Wait {class=BTTask_Wait, properties {waitTime {defaultValue=1}}}
 ]
 decorators []
 properties {
  children [
   {
    childComposite=&self:BTComposite_Selector_8
    childTask=null
    decorators [&self:BTDecorator_Blackboard_122]
   }
   {childComposite=null, childTask=&self:BTTask_Wait_0, decorators []}
  ]
  services []
 }
 services []
}
